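# Animation Blueprint for the TrickyDino pickaxe that adds secondary motion to its body and googly eyes.
# The input pose passes through weighted pupil constraints to two body bones, then low-alpha, X-only springs on three dyn_body bones, and is cached.
# Two duplicate AnimDynamics chains swing both eyes from that cache with angular springs and reduced gravity.
# An update-driven front-end flag selects World-space simulation in the front end and RootRelative otherwise.
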AnimBlueprint /BRCosmetics/Weapons/FORT_Melee/Pickaxe_TrickyDino/Meshes/TrickyDino_Axe_AnimBP
target_skeleton=&/TrickyDino_Axe_Skeleton
graphs [
 AnimGraph {nodes [
   n1 = Anim Dynamics - Bone: base_GooglyEye_l(ComponentPose=Local To Component(LocalPose=Use cached pose 'CacheStart'()), Alpha=1, ExternalForce="0,0,0") {
    angularDampingOverride=1
    angularSpringConstant=5000
    bAngularSpring=true
    bOverrideAngularDamping=true
    bUseGravityOverride=true
    gravityScale=0.35
    simulationSpace=RootRelative
    alphaBoolBlend {blendOption=Linear}
    boundBone {boneName=base_GooglyEye_l}
    componentAppliedLinearAccClamp=<3000 3000 3000>
    componentLinearAccScale=<1 1 1>
    componentLinearVelScale=<1 1 1>
    gravityOverride {z=-980}
    physicsBodyDefinitions [{
      boundBone {boneName=base_GooglyEye_l}
      boxExtents=<15 15 15>
      constraintSetup {angularLimitsMax=<5 75 75>, angularLimitsMin=<-5 -75 -75>, angularTarget=<1 0 0>}
      localJointOffset=<2 0 0>
     }]
   }
   n2 = Anim Dynamics - Bone: base_GooglyEye_r(ComponentPose=@n1, Alpha=1, ExternalForce="0,0,0") {
    angularDampingOverride=1
    angularSpringConstant=5000
    bAngularSpring=true
    bOverrideAngularDamping=true
    bUseGravityOverride=true
    gravityScale=0.35
    simulationSpace=RootRelative
    alphaBoolBlend {blendOption=Linear}
    boundBone {boneName=base_GooglyEye_r}
    componentAppliedLinearAccClamp=<3000 3000 3000>
    componentLinearAccScale=<1 1 1>
    componentLinearVelScale=<1 1 1>
    gravityOverride {z=-980}
    physicsBodyDefinitions [{
      boundBone {boneName=base_GooglyEye_r}
      boxExtents=<15 15 15>
      constraintSetup {angularLimitsMax=<5 75 75>, angularLimitsMin=<-5 -75 -75>, angularTarget=<1 0 0>}
      localJointOffset=<2 0 0>
     }]
   }
   n3 = Anim Dynamics - Bone: base_GooglyEye_l(ComponentPose=Local To Component(LocalPose=Use cached pose 'CacheStart'()), Alpha=1, ExternalForce="0,0,0") {
    angularDampingOverride=1
    angularSpringConstant=5000
    bAngularSpring=true
    bOverrideAngularDamping=true
    bUseGravityOverride=true
    gravityScale=0.35
    simulationSpace=World
    alphaBoolBlend {blendOption=Linear}
    boundBone {boneName=base_GooglyEye_l}
    componentAppliedLinearAccClamp=<3000 3000 3000>
    componentLinearAccScale=<1 1 1>
    componentLinearVelScale=<1 1 1>
    gravityOverride {z=-980}
    physicsBodyDefinitions [{
      boundBone {boneName=base_GooglyEye_l}
      boxExtents=<15 15 15>
      constraintSetup {angularLimitsMax=<5 75 75>, angularLimitsMin=<-5 -75 -75>, angularTarget=<1 0 0>}
      localJointOffset=<2 0 0>
     }]
   }
   n4 = Anim Dynamics - Bone: base_GooglyEye_r(ComponentPose=@n3, Alpha=1, ExternalForce="0,0,0") {
    angularDampingOverride=1
    angularSpringConstant=5000
    bAngularSpring=true
    bOverrideAngularDamping=true
    bUseGravityOverride=true
    gravityScale=0.35
    simulationSpace=World
    alphaBoolBlend {blendOption=Linear}
    boundBone {boneName=base_GooglyEye_r}
    componentAppliedLinearAccClamp=<3000 3000 3000>
    componentLinearAccScale=<1 1 1>
    componentLinearVelScale=<1 1 1>
    gravityOverride {z=-980}
    physicsBodyDefinitions [{
      boundBone {boneName=base_GooglyEye_r}
      boxExtents=<15 15 15>
      constraintSetup {angularLimitsMax=<5 75 75>, angularLimitsMin=<-5 -75 -75>, angularTarget=<1 0 0>}
      localJointOffset=<2 0 0>
     }]
   }
   Output Pose(Result=Blend Poses by bool(
     BlendPose_0=Component To Local(ComponentPose=@n4)
     BlendPose_1=Component To Local(ComponentPose=@n2)
     bActiveValue=Get IsFrontEndTrue() -> (IsFrontEndTrue:bool)
    ) {blendType=Linear}) {layerGroup=None}
   AnimGraphNode_SubInput_0 = Input Pose()
   n5 = "Constraint - base_GooglyPupil_l (PP)"!(ComponentPose=Local To Component(LocalPose=@AnimGraphNode_SubInput_0), Alpha=1, ConstraintWeights_0=0.6008, ConstraintWeights_1:float=0.3992) -> (Pose) {
    alphaBoolBlend {blendOption=Linear}
    boneToModify {boneName=base_GooglyPupil_l}
    constraintSetup [{targetBone {boneName=dyn_body_end}}, {targetBone {boneName=dyn_body_mid}}]
    constraintWeights [1,1]
   }
   n6 = "Constraint - base_GooglyPupil_r (PP)"!(ComponentPose=@n5, Alpha=1, ConstraintWeights_0=0.6008, ConstraintWeights_1:float=0.3992) -> (Pose) {
    alphaBoolBlend {blendOption=Linear}
    boneToModify {boneName=base_GooglyPupil_r}
    constraintSetup [{targetBone {boneName=dyn_body_end}}, {targetBone {boneName=dyn_body_mid}}]
    constraintWeights [1,1]
   }
   n7 = Spring controller - Bone: dyn_body!(ComponentPose:Component Space Pose Link Structure=@n6, Alpha:float=0.025) -> (Pose:Component Space Pose Link Structure) {
    is=title AnimGraphNode_SpringBone
    bLimitDisplacement=true
    bTranslateY=false
    bTranslateZ=false
    errorResetThresh=200
    maxDisplacement=100
    springDamping=8
    springStiffness=200
    alphaBoolBlend {blendOption=Linear}
    springBone {boneName=dyn_body}
   }
   n8 = Spring controller - Bone: dyn_body_mid!(ComponentPose:Component Space Pose Link Structure=@n7, Alpha:float=0.075) -> (Pose:Component Space Pose Link Structure) {
    is=title AnimGraphNode_SpringBone
    bLimitDisplacement=true
    bTranslateY=false
    bTranslateZ=false
    errorResetThresh=200
    maxDisplacement=100
    springDamping=8
    springStiffness=200
    alphaBoolBlend {blendOption=Linear}
    springBone {boneName=dyn_body_mid}
   }
   n9 = Spring controller - Bone: dyn_body_end!(ComponentPose:Component Space Pose Link Structure=@n8, Alpha:float=0.025) -> (Pose:Component Space Pose Link Structure) {
    is=title AnimGraphNode_SpringBone
    bLimitDisplacement=true
    bTranslateY=false
    bTranslateZ=false
    errorResetThresh=200
    maxDisplacement=100
    springDamping=8
    springStiffness=200
    alphaBoolBlend {blendOption=Linear}
    springBone {boneName=dyn_body_end}
   }
   Save cached pose 'CacheStart'(Pose=Component To Local(ComponentPose=@n9)) {cacheName=CacheStart}
  ]}
 EventGraph {nodes [
   Event BlueprintUpdateAnimation()
   > Set IsFrontEndTrue(IsFrontEndTrue:bool=Get bIsInFrontEnd()) -> (Output_Get:bool)
   TryGetPawnOwner()
  ]}
]
variables [IsFrontEndTrue {type=boolean}]
